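# Sound cue for a looping outdoor ambient bed of wind and distant thruster bleed at the hover platform lift, using box attenuation.
# It mixes a wind loop, shaped by looping pitch and volume envelopes for slow variation, with a quieter thruster loop.
# A random choice of start delay leads into a short fade-in.
# Only the first two quality levels play audio, and concurrency allows one instance.
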SoundCue /Game/Sounds/Artemis/Environment_Actors/HoverPlatform/Spot_HoverPlatform_Lift_Box_Wind_Thruster_Bleed_2D_Cue
root=Random_0
nodes [
 {id=Delay_0, children [Enveloper_1], properties {delayMax=4}}
 {id=Delay_1, children [Enveloper_1], properties {delayMax=6, delayMin=4}}
 {
  id=Enveloper_0
  children [WavePlayer_0]
  properties {
   bLoop=true
   bLoopIndefinitely=true
   loopEnd=30
   pitchCurve {editorCurveData {keys [{interpMode=RCIM_Cubic, time=0}, <at 15 0.9 RCIM_Cubic>, {interpMode=RCIM_Cubic, time=30}]}}
   volumeCurve {editorCurveData {keys [<at 0 1>]}}
  }
 }
 {
  id=Enveloper_1
  children [QualityLevel_0]
  properties {pitchCurve {editorCurveData {keys [{time=0}]}}, volumeCurve {editorCurveData {keys [<at 0 0 RCIM_Cubic>,<at 2 1 RCIM_Cubic>]}}}
 }
 {
  id=Enveloper_3
  children [Enveloper_0]
  properties {
   bLoop=true
   bLoopIndefinitely=true
   loopEnd=60
   pitchCurve {editorCurveData {keys [{time=0}]}}
   volumeCurve {editorCurveData {keys [
      {arriveTangent=-0.01, interpMode=RCIM_Cubic, time=0, value=1}
      <at 10 0.8 RCIM_Cubic>
      {arriveTangent=0.0095, interpMode=RCIM_Cubic, leaveTangent=0.0095, time=20, value=1}
      <at 31.1629 1>
      {arriveTangent=-0.0016, interpMode=RCIM_Cubic, leaveTangent=-0.0016, time=37.937, value=0.5372}
      <at 46.7125 0.975>
      <at 60 1 RCIM_Cubic>
     ]}}
  }
 }
 {id=Mixer_0, children [Enveloper_3, WavePlayer_12], properties {inputVolume [3,0.3]}}
 {id=QualityLevel_0, children [Mixer_0,Mixer_0,null,null]}
 {id=Random_0, children [Delay_0,Delay_1], properties {weights [1,1]}}
 {
  id=WavePlayer_0
  properties {bLooping=true, soundWaveAssetPtr=/Game/Sounds/Artemis/POI/Hollow_Base/Spots/Spot_Mountain_Wind_Window_Loop_Stereo}
 }
 {
  id=WavePlayer_12
  properties {bLooping=true, soundWaveAssetPtr=&/Spot_HoverPlatform_Thruster_Rip_Distant}
 }
]
properties {
 attenuationSettings=&/Spot_HoverPlatform_Thruster_Ripper_Wind_2D_Box_Att
 bOverrideConcurrency=true
 maxDistance=1280
 soundClassObject=&Fort_Ambient_Outdoors_Athena
 volumeMultiplier=0.501
 attenuationOverrides {falloffDistance=1600}
 concurrencyOverrides {maxCount=1, voiceStealReleaseTime=2}
}
