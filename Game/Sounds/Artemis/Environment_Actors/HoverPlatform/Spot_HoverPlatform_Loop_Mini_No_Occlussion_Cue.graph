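# Looping ambient cue for the mini hover platform spot sound, built from thruster and energy loops, with a random delay and volume fade-in at start.
# A quality switch selects either the full layered mix or a single thruster loop with overridden cone attenuation and reverb submix send.
# The full mix combines a random tonal thruster wave with thruster/energy blends under separate cone, sphere and distant-boost attenuations.
# Overridden concurrency caps instances at two, scaling volume by distance.
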SoundCue /Game/Sounds/Artemis/Environment_Actors/HoverPlatform/Spot_HoverPlatform_Loop_Mini_No_Occlussion_Cue
root=Delay_0
nodes [
 {
  id=Attenuation_0
  children [Random_0]
  properties {attenuationSettings=&/Spot_HoverPlatform_Energy_Loop_Close_No_Occ_Cone_Att}
 }
 {
  id=Attenuation_1
  children [Mixer_10]
  properties {attenuationSettings=&/Spot_HoverPlatform_Thruster_Loop_No_Occ_Cone_Att}
 }
 {
  id=Attenuation_2
  children [WavePlayer_1]
  properties {
   attenuationSettings=&Spot_HoverPlatform_Thruster_Loop_No_Occ_Cone_Att
   bOverrideAttenuation=true
   attenuationOverrides {
    attenuationShape=Cone
    bEnableFocusInterpolation=true
    bEnableListenerFocus=true
    bEnableReverbSend=false
    bEnableSubmixSends=true
    coneOffset=60
    coneSphereFalloffDistance=800
    coneSphereRadius=100
    falloffDistance=1500
    focusAttackInterpSpeed=0.1
    focusReleaseInterpSpeed=0.1
    nonFocusVolumeAttenuation=0.3
    stereoSpread=500
    attenuationShapeExtents=<2000 66 24>
    submixSendSettings [{
      bEnableHPFCutoff=false
      bEnableLPFCutoff=false
      disableManualSendClamp=false
      hPFCutoff=20
      lPFCutoff=20000
      maxSendDistance=7500
      maxSendLevel=0.25
      minSendDistance=0
      minSendLevel=0
      sendLevel=0.2
      sendLevelControlMethod=Linear
      soundSubmix=&Fortnite_Reverb_Distant
     }]
   }
  }
 }
 {
  id=Attenuation_3
  children [Mixer_10]
  properties {attenuationSettings=&/Spot_HoverPlatform_Distant_Thruster_Close_Custom_Boost_Att}
 }
 {
  id=Attenuation_8
  children [Mixer_15]
  properties {attenuationSettings=&/Spot_HoverPlatform_Energy_Loop_Close_No_Occ_Sphere_Att}
 }
 {id=Delay_0, children [Enveloper_0], properties {delayMax=1}}
 {
  id=Enveloper_0
  children [QualityLevel_0]
  properties {pitchCurve {editorCurveData {keys [{time=0}]}}, volumeCurve {editorCurveData {keys [<at 0 0>,<at 1 1>]}}}
 }
 {id=Mixer_0, children [Attenuation_0, Attenuation_1, Attenuation_3, Attenuation_8], properties {inputVolume [0.5,2,3,0.5]}}
 {id=Mixer_1, children [Attenuation_2], properties {inputVolume [2]}}
 {id=Mixer_10, children [WavePlayer_0,WavePlayer_4], properties {inputVolume [1,0.5]}}
 {id=Mixer_15, children [WavePlayer_18, WavePlayer_17], properties {inputVolume [2,1]}}
 {id=Mixer_2, children [WavePlayer_7], properties {inputVolume [0.9]}}
 {id=QualityLevel_0, children [Mixer_0,Mixer_0,Mixer_1,Mixer_1]}
 {id=Random_0, children [WavePlayer_5,WavePlayer_6,Mixer_2], properties {weights [1,1,1]}}
 {
  id=WavePlayer_0
  properties {bLooping=true, soundWaveAssetPtr=&/Spot_HoverPlatform_Thruster_Loop}
 }
 {
  id=WavePlayer_1
  properties {bLooping=true, soundWaveAssetPtr=&Spot_HoverPlatform_Thruster_Loop}
 }
 {
  id=WavePlayer_17
  properties {bLooping=true, soundWaveAssetPtr=&/Spot_HoverPlatform_Energy_Loop}
 }
 {
  id=WavePlayer_18
  properties {bLooping=true, soundWaveAssetPtr=&Spot_HoverPlatform_Thruster_Loop}
 }
 {
  id=WavePlayer_4
  properties {bLooping=true, soundWaveAssetPtr=&Spot_HoverPlatform_Energy_Loop}
 }
 {
  id=WavePlayer_5
  properties {bLooping=true, soundWaveAssetPtr=&/Spot_HoverPlatform_Thruster_Tonal_01}
 }
 {
  id=WavePlayer_6
  properties {bLooping=true, soundWaveAssetPtr=&/Spot_HoverPlatform_Thruster_Tonal_02}
 }
 {
  id=WavePlayer_7
  properties {bLooping=true, soundWaveAssetPtr=&/Spot_HoverPlatform_Thruster_Tonal_03}
 }
]
properties {
 bDebug=true
 bOverrideConcurrency=true
 maxDistance=6000
 soundClassObject=&Fort_Ambient_Outdoors_Athena
 volumeMultiplier=0.251
 attenuationOverrides {
  attenuationShape=Cone
  bEnableFocusInterpolation=true
  bEnableListenerFocus=true
  bEnableOcclusion=true
  coneSphereRadius=600
  distanceAlgorithm=Logarithmic
  falloffDistance=1000
  nonFocusVolumeAttenuation=0.7
  nonSpatializedRadiusStart=1500
  occlusionLowPassFilterFrequency=10000
  occlusionTraceChannel=ECC_GameTraceChannel11
  occlusionVolumeAttenuation=0.5
  stereoSpread=10000
  attenuationShapeExtents=<7500 8 10>
 }
 concurrencyOverrides {maxCount=2, voiceStealReleaseTime=2, volumeScaleAttackTime=2, volumeScaleMode=Distance}
}
