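# Sound cue for the distant, looping thruster rip ambience of a hover platform, in the outdoor ambient sound class.
# A quality-level node feeds its first two inputs from one shared looping wave player and leaves the other two empty.
# A dedicated capsule-named attenuation asset handles spatial falloff at low overall volume.
# Concurrency overrides cap it at two simultaneous voices, with distance-based volume scaling and long retrigger and voice-steal release times.
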SoundCue /Game/Sounds/Artemis/Environment_Actors/HoverPlatform/Spot_HoverPlatform_Thruster_Ripper_Capsule_Distant
root=QualityLevel_0
nodes [
 {id=QualityLevel_0, children [WavePlayer_12, WavePlayer_12, null, null]}
 {
  id=WavePlayer_12
  properties {bLooping=true, soundWaveAssetPtr=&/Spot_HoverPlatform_Thruster_Rip_Distant}
 }
]
properties {
 attenuationSettings=&/Spot_HoverPlatform_Distant_Thruster_Ripper_Capsule_Att
 bOverrideConcurrency=true
 maxDistance=13303.0449
 soundClassObject=&Fort_Ambient_Outdoors_Athena
 volumeMultiplier=0.251
 concurrencyOverrides {maxCount=2, retriggerTime=10, voiceStealReleaseTime=10, volumeScaleMode=Distance}
}
